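; Dirt bike vehicle Blueprint that adds water interaction effects and overrides explosion handling.
; Overlapping Rufus Water sets a per-lake height offset for splash and looping FX, which a repeating timer updates along with speed-scaled audio.
; On explosion, the dedicated server destroys parts, applies radial damage once and sends bush-destroyed events to bush-wearing players.
; Otherwise it plays explosion, wreck and surface-dependent Niagara effects with a radial force.
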(event EventBeginPlay
  (|CallParentFunction) ; This node is disabled and will not be called.\nDrag off pins to build functionality.
)

(event EventTick
  (params (DeltaSeconds :type Float%20%28single-precision%29))
  (|CallParentFunction DeltaSeconds) ; This node is disabled and will not be called.\nDrag off pins to build functionality.
)

(event Collision|EventActorBeginOverlap
  (params (OtherActor :type Actor%20Object%20Reference))
  ;[1 1 1] Handle water VFX offset on begin and end overlap with Rufus Water
  (bind _asbp_rufus_water (Utilities|Casting|CastToBP_Rufus_Water OtherActor))
  (Default|UpdateWaterVFXOffset true _asbp_rufus_water))

(event Collision|EventActorEndOverlap
  (params (OtherActor :type Actor%20Object%20Reference))
  ;[1 1 1] Handle water VFX offset on begin and end overlap with Rufus Water
  (bind _asbp_rufus_water (Utilities|Casting|CastToBP_Rufus_Water OtherActor))
  (Default|UpdateWaterVFXOffset false _asbp_rufus_water))

(event Vehicle|Water|EventOnEnteredWaterVolume
  (params (WaterSurfacePoint :type Vector%20%28by%20ref%29 :default "0, 0, 0"))
  (bind _returnvalue (CallFunction|GetWaterLevel))
  (bind (_returnvalue_x _returnvalue_y _returnvalue_z) (Math|Vector|BreakVector (Transformation|GetActorLocation)))
  (bind _watervfxoffset (Variables|Default|GetWaterVFXOffset))
  (bind _returnvalue_2 (+ _returnvalue _watervfxoffset))
  (bind _returnvalue_3 (- _returnvalue_2 _returnvalue_z))
  (bind _skeletalmesh (Variables|Vehicle|GetSkeletalMesh))
  ;[1 1 1] FX for water ineraction with Rufus Water
  (Networking|CanExecuteCosmeticEvents
    (:True
      (if (> (Rendering|GetRenderingDetailMode) 0)
        (bind _returnvalue_4 (Utilities|Time|SetTimerbyEvent (delegate Custom|CheckWaterLoop) 0.25 true))
        (bind _inlakefx (Variables|Default|GetInLakeFX)
          (:then
            (Components|DestroyComponent _inlakefx))
          (:else
            _))
        (bind _returnvalue_5 (Effects|SpawnEmitterAttached :EmitterTemplate "/Game/WaterAndWind/Particles/WaterInteraction/P_Vehicle_In_Water_Athena.P_Vehicle_In_Water_Athena" :AttachToComponent _skeletalmesh :AttachPointName "None" :Location (Math|Vector|MakeVector 0.0 0.0 _returnvalue_3) :Scale "1.000000,1.000000,1.000000" :LocationType "KeepRelativeOffset" :bAutoDestroy true :PoolingMethod "None" :bAutoActivate true)) ; Looping water FX
        (Variables|Default|SetInLakeFX _returnvalue_5)
        (Effects|SpawnEmitterAttached :EmitterTemplate "/Game/WaterAndWind/Particles/WaterInteraction/P_WaterSplash_Athena_Vehicle.P_WaterSplash_Athena_Vehicle" :AttachToComponent _skeletalmesh :AttachPointName "None" :Location (Math|Vector|MakeVector 0.0 0.0 _returnvalue_3) :Scale "1.000000,1.000000,1.000000" :LocationType "KeepRelativeOffset" :bAutoDestroy true :PoolingMethod "None" :bAutoActivate true) ; Splash burst FX
        (bind _inwaterloop (Variables|Default|GetInWaterLoop)
          (:then
            (Audio|Components|Audio|Stop _inwaterloop))
          (:else
            _))
        (Audio|SpawnSoundAttached "/Game/Athena/Sounds/Vehicles/GolfCart/GolfCart_Movement_Impact_Water_Cue.GolfCart_Movement_Impact_Water_Cue" (Variables|Vehicle|GetSkeletalMesh))
        (bind _returnvalue_6 (Audio|SpawnSoundAttached "/Game/Athena/Sounds/Vehicles/GolfCart/GolfCart_Movement_InWater_Loop_Cue.GolfCart_Movement_InWater_Loop_Cue" (Variables|Vehicle|GetSkeletalMesh)))
        (Variables|Default|SetInWaterLoop _returnvalue_6)))))

(event Custom|CheckWaterLoop
  (bind (_returnvalue_x _returnvalue_y _returnvalue_z) (Math|Vector|BreakVector (Transformation|GetActorLocation)))
  ;[1 1 1] FX for water ineraction with Rufus Water
  (bind _inlakefx (Variables|Default|GetInLakeFX)
    (:then
      (Transformation|SetRelativeLocation _inlakefx (Math|Vector|MakeVector 0.0 0.0 (- (+ (CallFunction|GetWaterLevel) (Variables|Default|GetWaterVFXOffset)) _returnvalue_z)) false false)
      (Effects|Components|ParticleSystem|SetVectorParameter (Variables|Default|GetInLakeFX) "Offset" (Math|Vector|Normalize (Math|Transform|InverseTransformDirection (Transformation|GetWorldTransform (Variables|Default|GetInLakeFX)) (Transformation|GetVelocity))))
      (bind _inwaterloop (Variables|Default|GetInWaterLoop)
        (:then
          (Audio|Components|Audio|SetVolumeMultiplier _inwaterloop (Math|Float|MapRangeClamped (Vehicle|GetForwardSpeedKmh) 0.0 2e+01 0.0 1.0) :!class /Script/Engine.AudioComponent))
        (:else)))
    (:else)))

(event Vehicle|Water|EventOnExitedWaterVolume
  (params (WaterSurfacePoint :type Vector%20%28by%20ref%29 :default "0, 0, 0"))
  ;[1 1 1] FX for water ineraction with Rufus Water
  (Networking|CanExecuteCosmeticEvents
    (:True
      (Utilities|Time|ClearandInvalidateTimerbyHandle (stale Vehicle|Water|EventOnEnteredWaterVolume _returnvalue_4))
      (bind _inlakefx (Variables|Default|GetInLakeFX)
        (:then
          (Components|Activation|SetActive _inlakefx false false)
          (bind _inwaterloop (Variables|Default|GetInWaterLoop)
            (:then
              (Audio|Components|Audio|FadeOut _inwaterloop 3.0 0.0 "Linear" :!class /Script/Engine.AudioComponent))
            (:else)))
        (:else)))))

(event EventHandleOnExplode
  (params (LastDamageInstigator :type Controller%20Object%20Reference) (Vehicle :type Actor%20Object%20Reference))
  (bind _surface_type_vehicle_is_on (Variables|SurfaceType|GetSurface%20Type%20Vehicle%20Is%20On))
  (bind _returnvalue (Math|Conversions|ToInteger(Byte) _surface_type_vehicle_is_on))
  ;[1 0.1458 0.1678] Override EVENT -> On Explode to be able to deal damage to Players wearing a bush item
  (Variables|PawnReferences|SetLast%20Damage%20Instigator LastDamageInstigator)
  (bind _asfort_athena_vehicle (Utilities|Casting|CastToFortAthenaVehicle Vehicle))
  (Variables|PawnReferences|SetVehicle _asfort_athena_vehicle)
  (Utilities|FlowControl|Sequence
    (:then_0
      (Development|PrintString (Utilities|GetDisplayName (Vehicle|GetChildActorbyClass 0)))
      ;;[0 0 0] Trigger Explosion Cosmetics
      (if (Networking|IsDedicatedServer)
        (CallFunction|DestroyAllDamageableParts)
        ;;[1 1 1] Damage is happening twice on server??
        (Utilities|FlowControl|DoOnce false
          (:Completed
            (bind _targeted_actors (Class|VehicleFunctionLibrary|VehicleExplodeandDamageNearbyActors (Variables|PawnReferences|GetVehicle) (Variables|PawnReferences|GetLast%20Damage%20Instigator) (ScalableFloat|GetValueatLevel (Variables|Hotfix|GetExplodeRadius)) (Variables|GameplayEffects|GetGE_ExplodeVehicleDamage) (Variables|GameplayEffects|GetGE_Explode_BuildingDamage) (Variables|GameplayEffects|GetGE_ExplodePlayerDamage) (ScalableFloat|GetValueatLevel (Variables|Hotfix|GetExplosionDamageVehicles)) (ScalableFloat|GetValueatLevel (Variables|Hotfix|GetExplosionDamagePlayers)) (ScalableFloat|GetValueatLevel (Variables|Hotfix|GetExplosionDamageBuildings)) false true (ScalableFloat|GetValueatLevel (Variables|Hotfix|GetExplodeVehicleImpulse)) (ScalableFloat|GetValueatLevel (Variables|Hotfix|GetExplodePlayerImpulse))))
            (Default|HandlePlayersinBush _targeted_actors)))
        (else
          (Class|MotorcycleFunctionLibrary|ExplosionFX (Variables|FX|GetExplodeFXNiagara) (Transformation|GetActorLocation) (Variables|Vehicle|GetSkeletalMesh) (Variables|FX|GetWreckFXNiagara) (Variables|CameraShake|GetExplosion_CameraShake_Radius) (CallFunction|GetMaxZeroHealthTime) (Variables|CameraShake|GetExplosionShake) (Variables|DamageStates|GetDestructionMaterial) 0.6)
          (bind _checkedphysicalsurfaceid (CallFunction|PhysicalSurfaceIDCheck _returnvalue))
          (if _checkedphysicalsurfaceid
            ;;[1 0.1458 0.1678] VFX Explosion  Phisical Surface Reaction
            (bind _returnvalue_2 (Niagara|SpawnSystemAttached "/Game/Effects/MudInteraction/NS_Physical_Surface_Explosion.NS_Physical_Surface_Explosion" (Variables|Vehicle|GetSkeletalMesh)))
            (Effects|Components|ParticleSystem|SetIntParameter _returnvalue_2 "PhysicalSurface" _returnvalue)
            (Effects|Components|ParticleSystem|SetFloatParameter _returnvalue_2 "Global Scale" 1.0)
            ;;;[0.2923 0.4974 1] Setting for Niagara Particle Data Export
            (Niagara|SetNiagaraVariable(Object) _returnvalue_2 "Export particle data" self))
          (CallFunction|SpawnFNRadialForce (Transformation|GetWorldTransform (Variables|Vehicle|GetSkeletalMesh)) 0.3 0.074 2.4e+03 1.75))))
    (:then_1
      (bind _damagefx (Variables|Default|GetDamageFX)
        (:then
          (Components|Activation|Deactivate _damagefx))
        (:else)))))

(fn HandlePlayersInBush
  (params (ActorArray :type Array%20of%20Actor%20Object%20References))
  ;[1 1 1] Check if actor is a player pawn and if they have the Bush Gameplay Tag send the Bush Destroyed Tag
  (for _array_element ActorArray
    (bind _asfort_player_pawn (Utilities|Casting|CastToFortPlayerPawn _array_element))
    (if (GameplayTags|HasMatchingGameplayTag _asfort_player_pawn "(TagName=\"Athena.Item.Bush\")")
      (Ability|SendGameplayEventtoActor _asfort_player_pawn "(TagName=\"Abilities.Generic.Athena.Bush.Destroyed\")"))))

(fn UpdateWaterVFXOffset
  (params (Entering :type Boolean) (Lake%20Reference :type BP%20Rufus%20Water%20Object%20Reference))
  (if Entering
    (Utilities|Casting|CastToBP_Rufus_Water_LeakyLake Lake%20Reference
      (:then
        (Development|PrintString "Vehicle is entering Leaky Lake.")
        (Variables|Default|SetWaterVFXOffset -56.0))
      (:CastFailed
        (Utilities|Casting|CastToBP_Rufus_Water_LootLake Lake%20Reference
          (:then
            (Development|PrintString "Vehicle is entering Loot Lake.")
            (Variables|Default|SetWaterVFXOffset 0.0))
          (:CastFailed
            (Development|PrintString "Vehicle is entering a Rufus Water Body.")
            (Variables|Default|SetWaterVFXOffset -56.0)))))
    (else
      (Development|PrintString "Vehicle is exiting water."))))

(fn ConstructionScript
  (|CallParentFunction))
